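# Post-process style AnimBP for the SpeakerBox backpack that passes its input pose through and toggles the "Hide Element 01" scalar on the mesh's dynamic material.
# Sets it when a SpeakerBox-tagged pickaxe is equipped or the front-end idle is active, and clears it while skydiving.
# Weapon, customization, skydiving and update events drive this.
# In the front end, a tagged character plus the FrontEnd_Idle_Bypass curve sets the idle flag.
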AnimBlueprint /BRCosmetics/Accessories/FORT_Backpacks/Backpack_M_MED_SpeakerBox/Meshes/M_MED_SpeakerBox_Pack_AnimBP
target_skeleton=&/M_MED_SpeakerBox_Pack_Skeleton
graphs [
 AnimGraph {nodes [Output Pose(Result=Input Pose()) {layerGroup=None}]}
 EventGraph {nodes [
   n1 = Event OnWeaponChanged()
   n2 = Get Dynamic Material Instance(execute=[@n1.then, @n12.then, @n8.then_0], SkeletalMeshComponent=GetPartSkeletalMeshComponent())
   n3 = PickaxeItemDefinitionHasAnyMetaTag(Tags=Make Array("[0]"=(TagName="Cosmetics.MetaTags.SpeakerBox")), bRequiresEquipped=true)
   > n4 = Branch(Condition=OR Boolean(A=Get FE Idle Active() -> (FE Idle Active:bool), B=@n3))
   > n5 = Branch(Condition=Get bIsSkyDiving())
   SetScalarParameterValue(execute=[@n6.then, @n5.then, @n7.else, @n4.else], self=@n2, ParameterName=Hide Element 01, Value=0)
   SetScalarParameterValue(execute=[@n5.else, @n7.then], self=@n2, ParameterName=Hide Element 01, Value=1)
   n6 = Event OnBeginSkydiving()
   Event OnEndSkydiving()
   > n7 = Branch(Condition=@n3)
   Event BlueprintUpdateAnimation()
   > n8 = Sequence()
   Branch(execute=@n8.then_1, Condition=Get bIsInFrontEnd())
   n9 = Make Array("[0]"=(TagName="Cosmetics.MetaTags.SpeakerBox"))
   n10 = GetCurveValue(self=GetOwnerFrontEndAnimInstance(), CurveName=FrontEnd_Idle_Bypass)
   > n11 = Branch(Condition=AND Boolean(A=CharacterItemDefinitionHasAnyMetaTag(Tags=@n9), B="float >= float"(A=@n10, B=1)))
   > Set FE Idle Active(FE Idle Active:bool=true) -> (Output_Get:bool)
   Set FE Idle Active(execute=@n11.else, FE Idle Active=false)
   n12 = Event OnCharacterCustomizationLoaded()
   TryGetPawnOwner()
  ]}
]
variables [FE Idle Active {type=boolean}]
